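; Sound library context for Chrome animation sounds that plays the event's sound on the Fortnite pawn that owns the library.
; Play spawns the sound attached to the pawn's root component and returns the resulting audio component.
; It sets the "PlayerAffliation" integer parameter (spelled that way) to distinguish the locally viewed pawn, friendly or neutral pawns, and hostile ones.
; When the actor is not a FortPawn, the sound plays in 2D only in the editor.
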(fn Play
  (params (InEventData :type Sound%20Library%20Context%20Event%20Input%20Structure%20%28by%20ref%29))
  (bind _sound (Utilities|Struct|BreakSoundLibraryContextEventInput InEventData))
  (bind _asfort_pawn (Utilities|Casting|CastToFortPawn (Variables|SoundLibrary|GetActor))
    (:then
      (bind _returnvalue (Audio|SpawnSoundAttached :Sound _sound :AttachToComponent (Variables|/Script/Engine.Actor|GetRootComponent _asfort_pawn) :AttachPointName "None" :LocationType "KeepRelativeOffset" :bStopWhenAttachedToDestroyed true :VolumeMultiplier 1.0 :PitchMultiplier 1.0 :StartTime 0.0 :bAutoDestroy true))
      (Audio|Parameter|SetIntegerParameter _returnvalue "PlayerAffliation" (Math|Integer|SelectInt 0 (Utilities|Select :Index (Team|GetTeamAffiliationforActortoLocalPlayer _asfort_pawn) :Friendly 1 :Neutral 1 :Hostile 2) (Pawn|IsLocallyViewed _asfort_pawn)) :!class /Script/Engine.AudioComponent)
      (return (Utilities|Array|MakeArray _returnvalue) true))
    (:CastFailed
      (if (Editor|InEditor)
        (Audio|PlaySound2D _sound)))))
